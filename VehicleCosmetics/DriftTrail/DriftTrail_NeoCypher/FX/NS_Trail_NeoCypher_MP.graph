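# Niagara system for the NeoCypher vehicle drift trail cosmetic, placed at wheel bones with a user-set offset.
# MainTrails001 streams two backward-moving ribbons, MeshesRotating001 pins rotating meshes and sprites to the bones, and SpritesMain001 emits short-lived streaking sprites.
# Each emitter resolves painted-variant colors, with bIsPainted also switching brightness and material values.
# Spawn groups and partitions vary look and bone assignment, while FN_DisableTireFX masks particles per wheel.
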NiagaraSystem /VehicleCosmetics/DriftTrail/DriftTrail_NeoCypher/FX/NS_Trail_NeoCypher_MP
dependencies {
 usedDataInterfaces [
  &NiagaraDataInterfaceArrayFloat3
  &NiagaraDataInterfaceCurve
  &NiagaraDataInterfaceSpriteRendererInfo
  &NiagaraDataInterfaceSkeletalMesh
 ]
 usedDynamicInputs []
 usedModules [
  &self:SystemSpawnScript.SystemScriptSource.SystemScriptGraph.NiagaraNodeAssignment_0.SetVariables_#1
  &SystemState
  &FN_UnpackVehicleFlags
  /Game/Effects/Niagara/Modules/Vehicle/FN_UpdateVehicleAttribs
  /Game/Effects/Niagara/Modules/Vehicle/FN_VehicleFlags_Cheatcodes
 ]
}
emitters {
 MainTrails001 {
  emitterSpawnScript [/SetVariables(Emitter.PainterSW:NiagaraFloat=&MakeCustomFloatFromBool(Bool=$User.bIsPainted, True Float=0, False Float=1)) {isSetParameters=true}]
  emitterUpdateScript [
   &EmitterState()
   &VC_RateQualityHandler(Epic Spawn Rate=30, High Spawn Rate=30, Medium Spawn Rate=25, Use Custom Rate?=true)
   &VC_SpawnRibbon_v1(Number of Ribbons=2, Spawn Time=-1, SpawnRate=$Emitter.Spawn Rate)
  ]
  particleSpawnScript [
   &VC_PaintedVariantsOverride(
    Black=<rgba 0.1146 0.0853 0.0853 1>
    BurntSienna=<rgba 0.4844 0.1701 0.0858 1>
    Cobalt=<rgba 0 0.1246 4.2 1>
    Crimson=<rgba 1.5 0.0312 0.0413 1>
    DefaultColor=<rgba 0.2392 24.2823 25 1>
    ForestGreen=<rgba 0 1.2 0.0606 1>
    Grey=<rgba 0.45 0.45 0.45 1>
    Lime=<rgba 1 2 0 1>
    Orange=<rgba 3 1.0464 0 1>
    Pink=<rgba 2 0.2396 1.3598 1>
    Purple=<rgba 1.6417 0 4 1>
    R1 Black:LinearColor=<rgba 0.004 0.003 0.003 1>
    R1 BurntSienna:LinearColor=<rgba 0.0365 0.0128 0.0065 1>
    R1 Cobalt:LinearColor=<rgba 0 0.008 0.2696 1>
    R1 Crimson:LinearColor=<rgba 0.0938 0.0029 0.0035 1>
    R1 ForestGreen:LinearColor=<rgba 0 0.1146 0.0058 1>
    R1 Grey:LinearColor=<rgba 0.004 0.003 0.003 1>
    R1 Lime:LinearColor=<rgba 0.1302 0.2604 0 1>
    R1 Orange:LinearColor=<rgba 0.1302 0.0454 0 1>
    R1 Pink:LinearColor=<rgba 0.2031 0.0243 0.1381 1>
    R1 Purple:LinearColor=<rgba 0.0962 0 0.2344 1>
    R1 Saffron:LinearColor=<rgba 0.224 0.1717 0 1>
    R1 SkyBlue:LinearColor=<rgba 0 0.0762 0.1146 1>
    R1 TitaniumWhite:LinearColor=<rgba 0.2188 0.2188 0.2188 1>
    Saffron=<rgba 2.5 1.9167 0 1>
    SkyBlue=<rgba 0 1.3308 2 1>
    TitaniumWhite=<rgba 4 4 4 1>
    Use R1 Painted Overrides=true
   )
   &VC_InitializeRibbon_v1(
    Color=$Particles.Color
    Custom Offset=true
    Enable Skeletal Position=true
    Lifetime=0.8
    Local Vector Offset=&AddVectorToPosition(Position=<-70 0 -20>, Vector=$User.offset)
    Offset Amount=1
    Ribbon Width:NiagaraFloat=&Multiply_Float(A=30, B=3)
   )
   &AddVelocity(Velocity=<-1000 0 0>) {module=AddVelocity001}
   /VehicleCosmetics/FX/Niagara/Modules/MergeEmitters/VC_ExtraBindingColors(
    ExtraColors=1
    R1 Color:LinearColor=&MakeCustomLinearColorFromBool(Bool=$User.bIsPainted, True Color=$User.PaintedColor_5, False Color=<rgba 1 1 1 1>)
   ) {enabled=false}
  ]
  particleUpdateScript [
   &ScaleColor(
    Scale Alpha=&MakeCustomFloatFromBool(Bool=$User.bIsPainted, True Float=1, False Float=1)
    Scale RGB=&MakeCustomVectorFromBool(Bool=$User.bIsPainted, True Vector=<10 10 10>, False Vector=<5 5 5>)
    ScaleA=true
    ScaleRGB=true
   )
   &InheritSourceMovement(
    Applied Position Scale=&Multiply_VectorByFloat(
     Vector=<1 1 1>
     Float=&FloatFromCurve(FloatCurve=DataInterface() {propertyValues {Curve {keys [
          {interpMode=RCIM_Cubic, tangentMode=RCTM_Break, time=0, value=1}
          {arriveTangent=-0.3257, interpMode=RCIM_Cubic, tangentMode=RCTM_Break, time=1, value=0.9}
         ]}}})
    )
    ApplyForce=false
    ApplyPosition=true
    ApplyPositionFalloff=false
   )
   &FN_DisableTireFX(Mask Index Attribute:NiagaraInt32=$Particles.SpawnGroup, Mask Index FL=2, Mask Index FR=3, Mask Index RL=0, Mask Index RR=1)
   &ParticleState()
   &SolveForcesAndVelocity(Clamp Velocity=false, Limit Acceleration=false)
   &DynamicMaterialParameters(
    Index 0 Param 1=1
    Index 0 Param 2=1
    Index 0 Param 3=1
    Index 0 Param 4=&MakeCustomFloatFromBool(Bool=$User.bIsPainted, True Float=40, False Float=6)
    Module.Index0Param0Write=true
    Module.Index0Param1Write=true
    Module.Index0Param2Write=true
    Module.Index0Param3Write=true
   )
   /SetVariables(
    Particles.CustomDynamicBinding:Vector4f=&MakeVector4(
     X=$User.Hardness
     Y=$User.Opacity
     Z=$User.AddMultiply
     W=&MakeCustomFloatFromBool(Bool=$User.bIsPainted, True Float=1, False Float=0)
    )
    "Particles.R1ColorScaled ":LinearColor=&VC_DynamicScaleColor(
     Color to Scale:LinearColor=$Particles.R1 Color
     RGB Scale:Vector3f=&MakeCustomVectorFromBool(Bool=$User.bIsPainted, True Vector=<4 4 4>, False Vector=<4 4 4>)
     Alpha Scale:NiagaraFloat=0.2
    )
   ) {isSetParameters=true}
   &VC_UpdateRibbon_v1(Detach Locked Particle=true, Enable Skeletal Position=true)
  ]
  rendererClasses [&NiagaraRibbonRendererProperties]
  renderers [
   {rendererClass=&NiagaraRibbonRendererProperties}
   {rendererClass=&NiagaraRibbonRendererProperties, rendererIndex=1}
  ]
 }
 MeshesRotating001 {
  emitterUpdateScript [
   &EmitterState(Life Cycle Mode=::NewEnumerator1, Loop Behavior=::NewEnumerator1, Loop Duration Mode=::NewEnumerator1, UseLoopDelay=false)
   &SpawnBurst_Instantaneous(Spawn Count=2, Use Spawn Probability=false, UseLoopCountLimit=false) {module=SpawnBurst_Instantaneous001}
   &SpawnBurst_Instantaneous(Spawn Count=2, Spawn Group=2, Use Spawn Probability=false, UseLoopCountLimit=false)
   &SpawnRate(SpawnRate=0, Use Spawn Probability=false)
   &SpawnBurst_Instantaneous(Spawn Count=2, Spawn Group=1, Use Spawn Probability=false, UseLoopCountLimit=false) {module=SpawnBurst_Instantaneous002}
  ]
  particleSpawnScript [
   &VC_PaintedVariantsOverride(
    Black=<rgba 0 0 0 8>
    BurntSienna=<rgba 0.0625 0.0314 0.0105 3>
    Cobalt=<rgba 0.028 0.0399 0.6719 1>
    Crimson=<rgba 0.2865 0.0014 0.0044 2>
    DefaultColor=<rgba 0 0.9063 0.9323 1>
    ForestGreen=<rgba 0 0.1146 0.0058 3>
    Grey=<rgba 0.035 0.035 0.035 2.5>
    Lime=<rgba 0.2656 0.5312 0 1.5>
    Orange=<rgba 0.5 0.1744 0 1.5>
    Pink=<rgba 0.4062 0.0487 0.2762 1.2>
    Purple=<rgba 0.1132 0.0012 0.5365 1>
    R1 Black:LinearColor=<rgba 0 0 0 2>
    R1 BurntSienna:LinearColor=<rgba 0.0642 0.0199 0.008 1.2783>
    R1 Cobalt:LinearColor=<rgba 0.032 0.0428 0.6146 1.6>
    R1 Crimson:LinearColor=<rgba 0.2083 0 0.0056 1>
    R1 DefaultColor=<rgba 0.369 0.0781 1 1>
    R1 ForestGreen:LinearColor=<rgba 0 0.2917 0.0147 1.5>
    R1 Grey:LinearColor=<rgba 0.035 0.035 0.035 1.5>
    R1 Lime:LinearColor=<rgba 0.138 0.276 0 1.5>
    R1 Orange:LinearColor=<rgba 0.6724 0.2346 0 1.2>
    R1 Pink:LinearColor=<rgba 0.4479 0.0537 0.3045 1.2>
    R1 Purple:LinearColor=<rgba 0.1758 0.0371 0.375 1>
    R1 Saffron:LinearColor=<rgba 0.349 0.2822 0 1>
    R1 SkyBlue:LinearColor=<rgba 0.0417 0.1976 0.276 1.25>
    R1 TitaniumWhite:LinearColor=<rgba 0.3438 0.3438 0.3438 1>
    Saffron=<rgba 0.7188 0.551 0 1>
    SkyBlue=<rgba 0 0.3119 0.4688 1>
    Use R1 Painted Overrides=true
   )
   /SetVariables(Particles.New VC Tag:NiagaraInt32=&SpawnGroup()) {isSetParameters=true}
   &InitializeParticle(
    Color=$Particles.Color
    Lifetime=$Emitter.CurrentLoopDuration
    Mesh Scale Mode=::NewEnumerator3
    Mesh Uniform Scale=&Multiply_Float(A=0.42, B=1)
    Sprite Size=<110 110>
    Sprite Size Mode=::NewEnumerator4
    UsePositionOffset=true
    Write Material Random=true
   )
   &PartitionParticles()
   &SkeletalMeshLocation(
    Bone / Socket Index=&Add_Integer(A=$Particles.PartitionParticles.Partition, B=4)
    Bone / Socket Sampling Mode=::NewEnumerator5
    Interpolate Mesh Sampling=false
    MaskSpawnGroup=false
    Mesh Sampling Type=::NewEnumerator6
    Sampled Position Coordinate Space=::Local
    Sampled Position Offset=&AddVectorToPosition(Position=<-70 0 -20>, Vector=$User.offset)
    Skeletal Mesh=$System.SkeletalMesh
   )
   &VC_ExtraBindingColors(
    ExtraColors=1
    R1 Color:LinearColor=&MakeCustomLinearColorFromBool(Bool=$User.bIsPainted, True Color=$User.PaintedColor_3, False Color=<rgba 0.0078 0.9647 0.9922 1>)
   ) {enabled=false}
   &AddVelocity(Velocity=&VC_Mask_Vectors(
     FrontEnd Value:Vector3f=<-1500 0 0>
     blsFrontEndPreview:NiagaraBool=false
     SpeedMin Value:Vector3f=<-800 0 0>
     SpeedMax Value:Vector3f=<-1 0 0>
     Speed Alpha:NiagaraFloat=$System.Speed_Alpha
    )) {enabled=false}
   /SetVariables(Particles.Scale02:Vector3f=&VectorFromFloat(Value=0.22)) {isSetParameters=true}
   &InitialMeshOrientation(
    Mesh Orientation Mode=::NewEnumerator7
    Rotation=&SelectVectorFromArray(
     Direct Array Index=&SpawnGroup()
     Vector Selection Array=DataInterface() {propertyValues {FloatData [<0 0 0>, <1 1 1>, <0.5 -0.3 0>]}}
    )
    Use Rotation Vector=true
   )
  ]
  particleUpdateScript [
   &FN_DisableTireFX(
    Mask Index Attribute:NiagaraInt32=$Particles.PartitionParticles.Partition
    Mask Index FL=2
    Mask Index FR=3
    Mask Index RL=0
    Mask Index RR=1
   )
   &ParticleState()
   &ScaleColor(
    Scale RGB=&MakeCustomVectorFromBool(
     Bool=$User.bIsPainted
     True Vector=/VehicleCosmetics/FX/Niagara/Modules/MergeEmitters/VC_VectorBySpawnGroup(
      SpawnGroup or ExecIndex:NiagaraInt32=0
      CustomInt:NiagaraInt32=0
      MaximumSpawnGroups:NiagaraInt32=3
      Vector Group 0:Vector3f=<4 4 4>
      Vector Group 1:Vector3f=<1.5 1.5 1.5>
      Vector Group 2:Vector3f=<4 4 4>
      Vector Group 3:Vector3f=<0 0 0>
      Vector Group 4:Vector3f=<0 0 0>
      Vector Group 5:Vector3f=<0 0 0>
     )
     False Vector=&VC_VectorBySpawnGroup(
      SpawnGroup or ExecIndex:NiagaraInt32=0
      CustomInt:NiagaraInt32=0
      MaximumSpawnGroups:NiagaraInt32=3
      Vector Group 0:Vector3f=<2 2 2>
      Vector Group 1:Vector3f=<1.5 1.5 1.5>
      Vector Group 2:Vector3f=<2 2 2>
      Vector Group 3:Vector3f=<0 0 0>
      Vector Group 4:Vector3f=<0 0 0>
      Vector Group 5:Vector3f=<0 0 0>
     )
    )
    ScaleA=true
    ScaleRGB=true
   )
   &SkeletalMeshLocation(
    Bone / Socket Index=&Add_Integer(A=$Particles.PartitionParticles.Partition, B=4)
    Bone / Socket Sampling Mode=::NewEnumerator5
    MaskSpawnGroup=false
    Mesh Sampling Type=::NewEnumerator6
    Orientation Sampling=::NewEnumerator1
    Sampled Position Offset=&AddVectorToPosition(Position=<-70 0 -20>, Vector=$User.offset)
    Skeletal Mesh=$System.SkeletalMesh
   ) {module=SkeletalMeshLocation001}
   &UpdateMeshOrientation(Rotation Rate=1.5, Rotation Vector=<0.7 0.5 0>)
   &SolveForcesAndVelocity(Clamp Velocity=false, Limit Acceleration=false)
   /SetVariables(Particles.Quat:Quat4f=<0.5 -0.3 0 1>) {enabled=false, isSetParameters=true}
  ]
  rendererClasses [&NiagaraMeshRendererProperties, &NiagaraSpriteRendererProperties]
  renderers [
   {rendererClass=&NiagaraMeshRendererProperties, rendererIndex=0}
   {rendererClass=&NiagaraMeshRendererProperties, rendererIndex=1}
   {rendererClass=&NiagaraSpriteRendererProperties, rendererIndex=2}
   {rendererClass=&NiagaraSpriteRendererProperties, rendererIndex=3}
  ]
 }
 SpritesMain001 {
  emitterUpdateScript [
   &EmitterState(Inactive Response=::NewEnumerator1, Life Cycle Mode=::NewEnumerator1, UseLoopDelay=false)
   &SpawnRate(Spawn Probability=0.6, SpawnRate=8, Use Spawn Probability=true)
   &SpawnBurst_Instantaneous(Spawn Count=2, Spawn Group=1, Use Spawn Probability=false, UseLoopCountLimit=false) {enabled=false}
   &SpawnBurst_Instantaneous(Spawn Count=2, Spawn Group=1, Use Spawn Probability=false, UseLoopCountLimit=false) {enabled=false, module=SpawnBurst_Instantaneous001}
  ]
  particleSpawnScript [
   &VC_PaintedVariantsOverride(
    Black=/VehicleCosmetics/FX/Niagara/Modules/MergeEmitters/VC_Dynamic_ColorBySpawnGroup(
     Color SG 0:LinearColor=<rgba 0 0 0 1>
     MaximumSpawnGroups:NiagaraInt32=2
     Color SG 1:LinearColor=<rgba 0 0 0 2>
     Color SG 2:LinearColor=<rgba 1 1 1 1>
     Color SG 3:LinearColor=<rgba 1 1 1 1>
     Color SG 4:LinearColor=<rgba 1 1 1 1>
    )
    BurntSienna=&VC_Dynamic_ColorBySpawnGroup(
     Color SG 0:LinearColor=<rgba 0.5573 0.2323 0.1451 2>
     MaximumSpawnGroups:NiagaraInt32=2
     Color SG 1:LinearColor=<rgba 0.0642 0.0199 0.008 1.2783>
     Color SG 2:LinearColor=<rgba 1 1 1 1>
     Color SG 3:LinearColor=<rgba 1 1 1 1>
     Color SG 4:LinearColor=<rgba 1 1 1 1>
    )
    Cobalt=&VC_Dynamic_ColorBySpawnGroup(
     Color SG 0:LinearColor=<rgba 0.0538 0.3838 0.8021 2>
     MaximumSpawnGroups:NiagaraInt32=2
     Color SG 1:LinearColor=<rgba 0.032 0.0428 0.6146 1.6>
     Color SG 2:LinearColor=<rgba 1 1 1 1>
     Color SG 3:LinearColor=<rgba 1 1 1 1>
     Color SG 4:LinearColor=<rgba 1 1 1 1>
    )
    Crimson=&VC_Dynamic_ColorBySpawnGroup(
     Color SG 0:LinearColor=<rgba 0.5469 0.1243 0.1034 2>
     MaximumSpawnGroups:NiagaraInt32=2
     Color SG 1:LinearColor=<rgba 0.2083 0 0.0056 1>
     Color SG 2:LinearColor=<rgba 1 1 1 1>
     Color SG 3:LinearColor=<rgba 1 1 1 1>
     Color SG 4:LinearColor=<rgba 1 1 1 1>
    )
    DefaultColor=&VC_Dynamic_ColorBySpawnGroup(
     Color SG 0:LinearColor=<rgba 0 0.688 1 5>
     MaximumSpawnGroups:NiagaraInt32=2
     Color SG 1:LinearColor=<rgba 0.369 0.0781 1 1>
     Color SG 2:LinearColor=<rgba 1 1 1 1>
     Color SG 3:LinearColor=<rgba 1 1 1 1>
     Color SG 4:LinearColor=<rgba 1 1 1 1>
    )
    ForestGreen=&VC_Dynamic_ColorBySpawnGroup(
     Color SG 0:LinearColor=<rgba 0.0903 0.6667 0.1194 3>
     MaximumSpawnGroups:NiagaraInt32=2
     Color SG 1:LinearColor=<rgba 0 0.2917 0.0147 1.5>
     Color SG 2:LinearColor=<rgba 1 1 1 1>
     Color SG 3:LinearColor=<rgba 1 1 1 1>
     Color SG 4:LinearColor=<rgba 1 1 1 1>
    )
    Grey=&VC_Dynamic_ColorBySpawnGroup(
     Color SG 0:LinearColor=<rgba 0.1094 0.1094 0.1094 1>
     MaximumSpawnGroups:NiagaraInt32=2
     Color SG 1:LinearColor=<rgba 0.035 0.035 0.035 1.5>
     Color SG 2:LinearColor=<rgba 1 1 1 1>
     Color SG 3:LinearColor=<rgba 1 1 1 1>
     Color SG 4:LinearColor=<rgba 1 1 1 1>
    )
    Lime=&VC_Dynamic_ColorBySpawnGroup(
     Color SG 0:LinearColor=<rgba 0.138 0.276 0 1.5>
     MaximumSpawnGroups:NiagaraInt32=2
     Color SG 1:LinearColor=<rgba 0.138 0.276 0 1.5>
     Color SG 2:LinearColor=<rgba 1 1 1 1>
     Color SG 3:LinearColor=<rgba 1 1 1 1>
     Color SG 4:LinearColor=<rgba 1 1 1 1>
    )
    Orange=&VC_Dynamic_ColorBySpawnGroup(
     Color SG 0:LinearColor=<rgba 0.75 0.3303 0.1055 2>
     MaximumSpawnGroups:NiagaraInt32=2
     Color SG 1:LinearColor=<rgba 0.6724 0.2346 0 1.2>
     Color SG 2:LinearColor=<rgba 1 1 1 1>
     Color SG 3:LinearColor=<rgba 1 1 1 1>
     Color SG 4:LinearColor=<rgba 1 1 1 1>
    )
    Pink=&VC_Dynamic_ColorBySpawnGroup(
     Color SG 0:LinearColor=<rgba 0.6667 0.184 0.4912 2>
     MaximumSpawnGroups:NiagaraInt32=2
     Color SG 1:LinearColor=<rgba 0.4479 0.0537 0.3045 1.2>
     Color SG 2:LinearColor=<rgba 1 1 1 1>
     Color SG 3:LinearColor=<rgba 1 1 1 1>
     Color SG 4:LinearColor=<rgba 1 1 1 1>
    )
    Purple=&VC_Dynamic_ColorBySpawnGroup(
     Color SG 0:LinearColor=<rgba 0.3361 0.0987 0.6771 2>
     MaximumSpawnGroups:NiagaraInt32=2
     Color SG 1:LinearColor=<rgba 0.1758 0.0371 0.375 1>
     Color SG 2:LinearColor=<rgba 1 1 1 1>
     Color SG 3:LinearColor=<rgba 1 1 1 1>
     Color SG 4:LinearColor=<rgba 1 1 1 1>
    )
    Saffron=&VC_Dynamic_ColorBySpawnGroup(
     Color SG 0:LinearColor=<rgba 0.7135 0.5871 0.052 2>
     MaximumSpawnGroups:NiagaraInt32=2
     Color SG 1:LinearColor=<rgba 0.349 0.2822 0 1>
     Color SG 2:LinearColor=<rgba 1 1 1 1>
     Color SG 3:LinearColor=<rgba 1 1 1 1>
     Color SG 4:LinearColor=<rgba 1 1 1 1>
    )
    SkyBlue=&VC_Dynamic_ColorBySpawnGroup(
     Color SG 0:LinearColor=<rgba 0.1177 0.5246 0.7292 2>
     MaximumSpawnGroups:NiagaraInt32=2
     Color SG 1:LinearColor=<rgba 0.0417 0.1976 0.276 1.25>
     Color SG 2:LinearColor=<rgba 1 1 1 1>
     Color SG 3:LinearColor=<rgba 1 1 1 1>
     Color SG 4:LinearColor=<rgba 1 1 1 1>
    )
    TitaniumWhite=&VC_Dynamic_ColorBySpawnGroup(
     Color SG 0:LinearColor=<rgba 0.8333 0.8333 0.8333 2>
     MaximumSpawnGroups:NiagaraInt32=2
     Color SG 1:LinearColor=<rgba 0.3438 0.3438 0.3438 1>
     Color SG 2:LinearColor=<rgba 1 1 1 1>
     Color SG 3:LinearColor=<rgba 1 1 1 1>
     Color SG 4:LinearColor=<rgba 1 1 1 1>
    )
   )
   /SetVariables(Particles.New VC Tag:NiagaraInt32=&SpawnGroup()) {isSetParameters=true}
   &InitializeParticle(
    Color=$Particles.Color
    Lifetime=/VehicleCosmetics/FX/Niagara/Modules/MergeEmitters/VC_FloatBySpawnGroup(
     SpawnGroup or ExecIndex:NiagaraInt32=0
     CustomInt:NiagaraInt32=0
     MaximumSpawnGroups:NiagaraInt32=2
     Float SG 0:NiagaraFloat=&MakeCustomFloatFromBool(
      Bool=$User.bIsFrontendPreview
      True Float=&RandomRangeFloat(Minimum=&Multiply_Float(A=0.2, B=1), Maximum=&Multiply_Float(A=0.4, B=1))
      False Float=&RandomRangeFloat(
       Minimum=&Multiply_Float(A=0.65, B=0.55)
       Maximum=&Multiply_Float(A=0.65, B=0.55)
      )
     )
     Float SG 1:NiagaraFloat=$Emitter.CurrentLoopDuration
     Float SG 2:NiagaraFloat=$Emitter.CurrentLoopDuration
     Float SG 3:NiagaraFloat=0
     Float SG 4:NiagaraFloat=0
     Float SG 5:NiagaraFloat=0
    )
    Sprite Rotation Angle=&VC_FloatBySpawnGroup(
     SpawnGroup or ExecIndex:NiagaraInt32=0
     CustomInt:NiagaraInt32=0
     MaximumSpawnGroups:NiagaraInt32=2
     Float SG 0:NiagaraFloat=180
     Float SG 1:NiagaraFloat=0
     Float SG 2:NiagaraFloat=0
     Float SG 3:NiagaraFloat=0
     Float SG 4:NiagaraFloat=0
     Float SG 5:NiagaraFloat=0
    )
    Sprite Rotation Mode=::NewEnumerator0
    Sprite Size=<110 110>
    Sprite Size Mode=::NewEnumerator4
    UsePositionOffset=false
    Write Material Random=true
   )
   &PartitionParticles()
   &SkeletalMeshLocation(
    Bone / Socket Index=&Add_Integer(A=$Particles.PartitionParticles.Partition, B=4)
    Bone / Socket Sampling Mode=::NewEnumerator5
    Interpolate Mesh Sampling=false
    MaskSpawnGroup=false
    Mesh Sampling Type=::NewEnumerator6
    Sampled Position Offset=&AddVectorToPosition(Position=<-70 0 -20>, Vector=$User.offset)
    Skeletal Mesh=$System.SkeletalMesh
   )
   &AddVelocity(
    Velocity=&SelectVectorFromArray(
     Direct Array Index=&SpawnGroup()
     Vector Selection Array=DataInterface() {propertyValues {FloatData [<-1000 0 0>, <0 0 0>]}}
    )
    Velocity Speed Scale=&VC_FloatBySpawnGroup(
     SpawnGroup or ExecIndex:NiagaraInt32=0
     CustomInt:NiagaraInt32=0
     MaximumSpawnGroups:NiagaraInt32=2
     Float SG 0:NiagaraFloat=1
     Float SG 1:NiagaraFloat=0
     Float SG 2:NiagaraFloat=0
     Float SG 3:NiagaraFloat=0
     Float SG 4:NiagaraFloat=0
     Float SG 5:NiagaraFloat=0
    )
   )
   &VC_ExtraBindingColors(
    ExtraColors=1
    R1 Color:LinearColor=&MakeCustomLinearColorFromBool(
     Bool=$User.bIsPainted
     True Color=$User.PaintedColor_6
     False Color=&RandomRangeLinearColor(Minimum=<rgba 0.0078 0.9647 0.9922 1>, Maximum=<rgba 0.4 0.0863 0.9412 1>)
    )
   )
  ]
  particleUpdateScript [
   &FN_DisableTireFX(
    Mask Index Attribute:NiagaraInt32=$Particles.PartitionParticles.Partition
    Mask Index FL=2
    Mask Index FR=3
    Mask Index RL=0
    Mask Index RR=1
   )
   &ParticleState()
   &ScaleColor(
    Scale Alpha=&VC_FloatBySpawnGroup(
     SpawnGroup or ExecIndex:NiagaraInt32=0
     CustomInt:NiagaraInt32=0
     MaximumSpawnGroups:NiagaraInt32=2
     Float SG 0:NiagaraFloat=1
     Float SG 1:NiagaraFloat=1
     Float SG 2:NiagaraFloat=0
     Float SG 3:NiagaraFloat=0
     Float SG 4:NiagaraFloat=0
     Float SG 5:NiagaraFloat=0
    )
    Scale RGB=&VC_VectorBySpawnGroup(
     SpawnGroup or ExecIndex:NiagaraInt32=0
     CustomInt:NiagaraInt32=0
     MaximumSpawnGroups:NiagaraInt32=2
     Vector Group 0:Vector3f=<2 2 2>
     Vector Group 1:Vector3f=<1.5 1.5 1.5>
     Vector Group 2:Vector3f=<0 0 0>
     Vector Group 3:Vector3f=<0 0 0>
     Vector Group 4:Vector3f=<0 0 0>
     Vector Group 5:Vector3f=<0 0 0>
    )
    ScaleA=true
    ScaleRGB=true
   )
   &InheritSourceMovement(
    Applied Position Scale=&Multiply_VectorByFloat(
     Vector=<1 1 1>
     Float=&FloatFromCurve(FloatCurve=DataInterface() {propertyValues {Curve {keys [
          {interpMode=RCIM_Cubic, tangentMode=RCTM_Break, time=0, value=1}
          {arriveTangent=-0.3257, interpMode=RCIM_Cubic, tangentMode=RCTM_Break, time=1, value=0.9}
         ]}}})
    )
    ApplyForce=false
    ApplyPosition=true
    ApplyPositionFalloff=false
   )
   &SolveForcesAndVelocity(Clamp Velocity=false, Limit Acceleration=false)
   /SetVariables(
    Particles.R1 Color:LinearColor=&VC_DynamicScaleColor(
     Color to Scale:LinearColor=$Particles.Initial.R1 Color
     RGB Scale:Vector3f=&MakeCustomVectorFromBool(Bool=$User.bIsPainted, True Vector=<0 0 0>, False Vector=<0 0 0>)
     Alpha Scale:NiagaraFloat=&FloatFromCurve(
      FloatCurve=DataInterface() {propertyValues {Curve {keys [
          {interpMode=RCIM_Cubic, leaveTangent=-1.1247, tangentMode=RCTM_Break, time=0, value=3}
          {arriveTangent=-10.3511, interpMode=RCIM_Cubic, leaveTangent=-10.3811, tangentMode=RCTM_Break, time=0.3, value=2}
          {interpMode=RCIM_Cubic, tangentMode=RCTM_Break, time=0.7, value=0}
         ]}}}
      Scale Curve=0
     )
    )
    Particles.Size02:Vector2f=&RandomRangeVector2D(
     Minimum=&Multiply_Vector2D_ByFloat(Float=1.5, Vector2D=<6 8>)
     Maximum=&Multiply_Vector2D_ByFloat(Float=1.5, Vector2D=<8 30>)
    )
    Particles.Velocity02:Vector3f=&MakeCustomVectorFromBool(Bool=$User.bIsFrontendPreview, True Vector=<0 0 0>, False Vector=<0 0 0>)
   ) {isSetParameters=true}
   &DynamicMaterialParameters(
    Index 0 Param 1=5
    Module.Index0Param0Write=true
    Module.Index0Param1Write=true
    Module.Index0Param2Write=true
    Module.Index0Param3Write=true
   )
   &SubUVAnimation(
    End Frame Range Override:NiagaraInt32=&RandomRangeInt(Minimum=0, Maximum=1)
    Output Alternate Renderer Binding=true
    "Sprite Renderer "=DataInterface() {propertyValues {SpriteRenderer=&self:SpritesMain001_0.NiagaraSpriteRendererProperties_1}}
    Start Frame Range Override=&RandomRangeInt(Minimum=0, Maximum=1)
    SubUV Animation Mode:ENiagara_SubUVLookupModeV2=::NewEnumerator2
    UseEndFrame=true
    UseStartFrame=true
   )
   &SkeletalMeshLocation(
    Bone / Socket Index=&Add_Integer(A=$Particles.PartitionParticles.Partition, B=4)
    Bone / Socket Sampling Mode=::NewEnumerator5
    MaskSpawnGroup=false
    Mesh Sampling Type=::NewEnumerator6
    Sampled Position Offset=&AddVectorToPosition(Position=<-70 100 -20>, Vector=$User.offset)
    Skeletal Mesh=$System.SkeletalMesh
   ) {enabled=false, module=SkeletalMeshLocation001}
  ]
  rendererClasses [&NiagaraSpriteRendererProperties]
  renderers [
   {rendererClass=&NiagaraSpriteRendererProperties}
   {rendererClass=&NiagaraSpriteRendererProperties, rendererIndex=1}
  ]
 }
}
summary {
 systemName=NS_Trail_NeoCypher_MP
 emitters [
  {
   emitterName=MeshesRotating001
   rendererClasses [&NiagaraMeshRendererProperties, &NiagaraSpriteRendererProperties]
  }
  {emitterName=SpritesMain001, rendererClasses [&NiagaraSpriteRendererProperties]}
  {emitterName=MainTrails001, rendererClasses [&NiagaraRibbonRendererProperties]}
 ]
 userVariables [
  User.AddMultiply {type=NiagaraFloat, defaultValue=0, description=""}
  User.bIsFrontendPreview {type=NiagaraBool}
  User.bIsPainted {type=NiagaraBool}
  User.ColorVariantID {type=NiagaraInt32}
  User.Hardness {type=NiagaraFloat, defaultValue=3, description=""}
  User.offset {type=Vector3f, description="offset from wheels(set through cosmetic template)", defaultValue {struct=Vector3f, value=<0 0 0>}}
  User.Opacity {type=NiagaraFloat, defaultValue=3, description=""}
  User.PaintedColor {type=LinearColor, description="", defaultValue {struct=LinearColor, value=<rgba 1 1 1 1>}}
  User.PaintedColor_2 {type=LinearColor, description="", defaultValue {struct=LinearColor, value=<rgba 1 1 1 1>}}
  User.PaintedColor_3 {type=LinearColor, description="", defaultValue {struct=LinearColor, value=<rgba 1 1 1 1>}}
  User.PaintedColor_4 {type=LinearColor, description="", defaultValue {struct=LinearColor, value=<rgba 1 1 1 1>}}
  User.PaintedColor_5 {type=LinearColor, description="", defaultValue {struct=LinearColor, value=<rgba 1 1 1 1>}}
  User.PaintedColor_6 {type=LinearColor, description="", defaultValue {struct=LinearColor, value=<rgba 1 1 1 1>}}
  User.PaintedScalar {type=NiagaraFloat, defaultValue=1, description=""}
  User.VehicleFlags {type=NiagaraInt32, defaultValue=806289408}
  User.Wheel_Radius_Array {
   type=NiagaraDataInterfaceArrayFloat
   defaultValue {
    struct=NiagaraExt_VariableValue_DataInterface
    value {
     dataInterface=&self:NiagaraDataInterfaceArrayFloat_0
     dataInterfaceClass=&NiagaraDataInterfaceArrayFloat
    }
   }
  }
  User.Wheel_Width_Array {
   type=NiagaraDataInterfaceArrayFloat
   defaultValue {
    struct=NiagaraExt_VariableValue_DataInterface
    value {
     dataInterface=&self:NiagaraDataInterfaceArrayFloat_1
     dataInterfaceClass=&NiagaraDataInterfaceArrayFloat
    }
   }
  }
 ]
}
