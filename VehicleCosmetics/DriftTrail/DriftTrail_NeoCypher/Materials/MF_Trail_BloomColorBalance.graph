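# Material function for the NeoCypher drift trail that rebalances a custom paint color, apparently so bloom stays even across colors.
# It sorts the color by RGB sum: light colors scale by a white-paint brightness, mid colors are normalized and scaled by about 0.577, and dark colors scale by a black-paint darkness.
# HasBlackPaintVersion toggles the dark tier.
# UsedForParticles replaces the thresholds with a dark-to-normalized lerp driven by the saturated sum.
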MaterialFunction /VehicleCosmetics/DriftTrail/DriftTrail_NeoCypher/Materials/MF_Trail_BloomColorBalance
graph {nodes [
  n1 = FunctionInput(Preview=Constant3Vector() {constant=<rgba 1 1 1 0>}) {inputName=CustomColor}
  n2 = Multiply(A=@n1, B=FunctionInput(Preview=Constant() {r=0.5}) {inputName=BlackPaint_Darkness, inputType=FunctionInput_Scalar})
  n3 = Add(A=Add(A=ComponentMask(None=@n1) {r=true}, B=ComponentMask(None=@n1) {g=true}), B=ComponentMask(None=@n1) {b=true})
  MaterialExpressionConstantClamp_26 = Saturate(None=@n3)
  n4 = Multiply(A=Constant3Vector() {constant=<rgba 0.5774 0.5774 0.5774 0>}, B=Normalize(VectorInput=@n1))
  n5 = Constant() {r=2}
  n6 = Multiply(A=FunctionInput(Preview=Constant() {r=1}) {inputName=WhitePaint_Brightness, inputType=FunctionInput_Scalar}, B=@n1)
  n7 = If(A=@n3, B=@n5, "A > B"=@n6, "A == B"=@n6, "A < B"=If(A=@n3, B=Constant() {r=0.2}, "A > B"=@n4, "A == B"=@n2, "A < B"=@n2))
  n8 = If(A=@n3, B=@n5, "A > B"=@n6, "A == B"=@n6, "A < B"=@n4)
  n9 = FunctionInput(Preview=StaticBool() {value=true}) {inputName=HasBlackPaintVersion, inputType=FunctionInput_StaticBool}
  n10 = StaticBool()
  FunctionOutput(None=StaticSwitch(
    True=LinearInterpolate(A=@n2, B=@n4, Alpha=@MaterialExpressionConstantClamp_26)
    False=StaticSwitch(True=@n7, False=@n8, Value=@n9)
    Value=FunctionInput(Preview=@n10) {inputName=UsedForParticles, inputType=FunctionInput_StaticBool}
   ))
 ]}
